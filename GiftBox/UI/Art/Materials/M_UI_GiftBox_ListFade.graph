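# Translucent UI material for GiftBox, named as a list fade, that builds an eased gradient over the widget.
# Normalized UI UVs pass through a Rotator driven by the Rotation parameter, scaled so one unit is a full turn.
# The rotated U channel feeds the MF_UI_EaseCurves function to shape the gradient.
# A GetSlatePost1 call is also present, and no connection to the material outputs appears in the graph.
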Material /GiftBox/UI/Art/Materials/M_UI_GiftBox_ListFade
graph {
 properties {
  bAutomaticallySetUsageInEditor=false
  bUsedWithStaticMesh=false
  blendMode=BLEND_Translucent
  materialDomain=MD_UI
  usedShadingModels=Default Lit
 }
 nodes [
  n1 = MaterialFunctionCall() {materialFunction=&GetUserInterfaceUV}
  n2 = ScalarParameter() {parameterName=Rotation}
  MaterialFunctionCall("Value (0-1)"=ComponentMask(None=Rotator(Coordinate=@n1.Normalized UV, Time=@n2) {speed=6.2832}) {r=true}) {materialFunction=&MF_UI_EaseCurves}
  MaterialFunctionCall!("UVs (Optional)") {materialFunction=/Engine/Functions/UserInterface/GetSlatePost1}
 ]
}
