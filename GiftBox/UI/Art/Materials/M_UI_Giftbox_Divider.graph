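# M_UI_Giftbox_Divider is a translucent UI material for a divider in the GiftBox interface, going by its name.
# It pairs a constant white color with a shape mask from the shared MF_UI_CoreUI_RoundedMask function, fed a constant corner radius.
# The mask is scaled down by a constant multiplier, which suggests a soft, semi-transparent rounded bar.
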Material /GiftBox/UI/Art/Materials/M_UI_Giftbox_Divider
graph {
 properties {
  bAutomaticallySetUsageInEditor=false
  bUsedWithStaticMesh=false
  blendMode=BLEND_Translucent
  materialDomain=MD_UI
  usedShadingModels=Default Lit
 }
 nodes [
  Constant4Vector() {constant=<rgba 1 1 1 0>}
  Multiply(A=MaterialFunctionCall(Corner Radius=Constant() {r=0.5}) {materialFunction=&MF_UI_CoreUI_RoundedMask}.Shape Mask) {constB=0.6}
 ]
}
