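# UI-domain Modulate material for the Nexus gradient overlay, multiplying what lies behind it to darken it along a vertical gradient.
# Inverted normalized UI UV plus inverted Offset shifts the gradient; its green channel blends from one minus Dark Amount toward white, darkest toward the bottom.
# A final lerp by vertex alpha runs from the default black to that gradient, so lower alpha darkens rather than fades the overlay.
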Material /GiftBox/UI/Art/Materials/M_UI_Nexus_GradientOverlay
graph {
 properties {
  bAutomaticallySetUsageInEditor=false
  bUsedWithStaticMesh=false
  blendMode=BLEND_Modulate
  materialDomain=MD_UI
  usedShadingModels=Default Lit
 }
 nodes [
  n1 = ScalarParameter() {defaultValue=1, parameterName=Dark Amount}
  n2 = Add(
   A=OneMinus(None=MaterialFunctionCall() {materialFunction=&GetUserInterfaceUV}.Normalized UV)
   B=OneMinus(None=ScalarParameter() {defaultValue=1, parameterName=Offset, sliderMax=2, sliderMin=0.5})
  ) {constB=-0.5}
  LinearInterpolate(B=LinearInterpolate(A=OneMinus(None=@n1), Alpha=ComponentMask(None=@n2) {g=true}), Alpha=VertexColor().A)
 ]
}
